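; Manager that places the Season 10 Figment floating island by injecting its level at the position carried by the island travel event.
; The authority listens for that stateful event, then SpawnIsland caches the transform, removes the previously injected level, async-loads the injection data and stores the new level GUID.
; UI helper data OnRep spawns or reuses a local helper actor for the map icon where cosmetics run.
; TestSpawnIsland spawns at the view target's location.
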(event EventBeginPlay
  (Utilities|SwitchHasAuthority
    (:Authority
      ;[1 1 1] Listen to Island Travel Event
      (bind (_asynctaskproxy _payload) (GameplayEventRouter|StartListeningToStatefulEvent (GameplayEventRouter|GetGlobalEventRouter) "/Figment_S10_LevelInstances/BP_FloatingIsland/Figment_FloatingIslandTravel_Event.Figment_FloatingIslandTravel_Event")
        (:OnEventReceived
          _)
        (:OnSavedEventStateExists
          _))
      (if (Utilities|Struct|BreakFigmentFloatingIslandTravelEvent _payload)
        (CallFunction|SpawnIsland (.IslandPosition_7_9DFB03BF49E988C841F253A7F65429A5 (Utilities|Struct|BreakFigmentFloatingIslandTravelEvent _payload)))))))

(event Custom|SpawnIsland
  (params (Location :type Vector))
  (bind _output_get (Variables|Default|SetCachedIslandTransform (Math|Transform|MakeTransform Location)))
  (Variables|SetUIHelperData_CachedIslandTransform_2_FF14E2C54151C18894DB28BB82FA1151 (Utilities|Struct|MakeUIHelperDataS10 _output_get))
  (CallFunction|RemoveInjectedLevel (Variables|Default|GetInjected%20Level%20GUID))
  (bind _object (Utilities|AsyncLoadAsset (Variables|Default|GetLevel%20Injection%20Data))
    (:Completed
      (bind _asfort_level_injection_data_base (Utilities|Casting|CastToFortLevelInjectionDataBase _object))
      (bind _injectedlevelguid (CallFunction|RequesttoInjectLevel (Utilities|ResolveSoftReference _asfort_level_injection_data_base) (Variables|Default|GetCachedIslandTransform)))
      (Variables|Default|SetInjected%20Level%20GUID _injectedlevelguid))))

(event Custom|TestSpawnIsland
  (CallFunction|SpawnIsland (Transformation|GetActorLocation (Pawn|GetViewTarget (Game|GetPlayerController 0)))))

(fn DestroyUIHelperActor
  (bind _ui_helper_actor (Variables|Default|GetUI%20Helper%20Actor))
  (Utilities|IsValid _ui_helper_actor
    (:Is%20Valid
      (Actor|DestroyActor _ui_helper_actor)
      (Variables|Default|SetUI%20Helper%20Actor 0))))

(fn OnRep_UIHelperData
  (bind _uihelperdata_cachedislandtransform_2_ff14e2c54151c18894db28bb82fa1151 (Utilities|Struct|BreakUIHelperDataS10 (Variables|GetUIHelperData_CachedIslandTransform_2_FF14E2C54151C18894DB28BB82FA1151_Location)))
  (bind _uihelperdata_cachedislandtransform_2_ff14e2c54151c18894db28bb82fa1151_location (Math|Transform|BreakTransform _uihelperdata_cachedislandtransform_2_ff14e2c54151c18894db28bb82fa1151))
  (Networking|CanExecuteCosmeticEvents
    (:True
      (bind _ui_helper_actor (Variables|Default|GetUI%20Helper%20Actor)
        (:then
          _)
        (:else
          (bind _returnvalue (Game|SpawnActorfromClass "/Figment_S10_LevelInstances/BP_FloatingIsland/BP_FloatingIsland_UIHelper_Actor.BP_FloatingIsland_UIHelper_Actor_C" (struct :Location "0, 0, 0" :Rotation "0, 0, 0" :Scale "1.000000,1.000000,1.000000") "AlwaysSpawn" "OverrideRootScale" 0 0))
          (bind _output_get (Variables|Default|SetUI%20Helper%20Actor _returnvalue))
          (Utilities|IsValid _output_get
            (:Is%20Valid
              _))))
      (Class|BPFloatingIslandUIHelperActor|SetupIslandMapIcon _ui_helper_actor _uihelperdata_cachedislandtransform_2_ff14e2c54151c18894db28bb82fa1151_location))))
