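; Helper actor for the Season 10 floating island that places the island's icon on the Fortnite Athena map.
; SetupIslandMapIcon stores a target position, clears any existing icon, enables the shadow layer and selects the V1_2 island icon.
; It registers that icon as dynamic land on the game state with a rotation offset.
; The actor's object name keys both adding and removing the land entry.
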(event Custom|SetupIslandMapIcon
  (params (TargetPosition :type Vector))
  (Variables|Default|SetTargetPosition TargetPosition)
  (CallFunction|ClearMapIcons)
  (Variables|Default|SetUseShadowLayer true)
  (Variables|Default|SetCurrentIslandIcon (Variables|Default|GetIslandIconV1_2))
  (Class|FortGameStateAthena|SetDynamicLand (Fortnite|Athena|GameState|GetGameStateAthena) (Utilities|String|StringToName (Utilities|GetObjectName self)) (Variables|Default|GetTargetPosition) (Variables|Default|GetLandIconRotationOffset) (Variables|Default|GetCurrentIslandIcon) (Variables|Default|GetUseShadowLayer)))

(fn ClearMapIcons
  (Class|FortGameStateAthena|RemoveDynamicLand (Fortnite|Athena|GameState|GetGameStateAthena) (Utilities|String|StringToName (Utilities|GetObjectName self)) true))

(fn ConstructionScript)
