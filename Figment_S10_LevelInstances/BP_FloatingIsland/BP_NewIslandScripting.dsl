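; Floating island scripting Blueprint that announces island travel and places its map ping during a calendar event.
; BeginPlay calls GetIslandPosition, which samples the spline at Progress and snaps it to a grid.
; After playlist and match readiness, an active event sets Progress from its timespan ratio, after that position call.
; It then broadcasts a stateful travel event and sets MapPingLocation, whose OnRep moves the ping collision sphere.
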(event EventBeginPlay
  (bind _returnvalue (CallFunction|GetIslandPosition))
  (Class|FortAsyncActionCurrentPlaylistReady|PlaylistDataReadyAsync
    (:OnReady
      (Class|FortAsyncActionAuthorityMatchReady|MatchReadyAsync
        (:OnReady
          (bind (_biseventactive _timesincebegin _timeuntilend _timespanratio) (CalendarEvents|GetCalendarEventInformation (Variables|Default|GetCalendarEvent)))
          (if _biseventactive
            (bind _output_get (Variables|Default|SetProgress _timespanratio))
            (Variables|Default|SetMapPingLocation _returnvalue)
            (GameplayEventRouter|BroadcastStatefulEvent (GameplayEventRouter|GetGlobalEventRouter) (Utilities|Struct|MakeFigmentFloatingIslandTravelEvent true _returnvalue _output_get) 0)))))))

(fn GetIslandPosition
  (return (Math|Vector|MakeVector (Math|Float|SnapToGrid(Float) (.x (Spline|GetLocationatDistanceAlongSpline (Variables|Default|GetSpline) (* (Spline|GetSplineLength (Variables|Default|GetSpline)) (Variables|Default|GetProgress)) "World")) 512.0) (Math|Float|SnapToGrid(Float) (.y (Spline|GetLocationatDistanceAlongSpline (Variables|Default|GetSpline) (* (Spline|GetSplineLength (Variables|Default|GetSpline)) (Variables|Default|GetProgress)) "World")) 512.0) (Math|Float|SnapToGrid(Float) (.z (Spline|GetLocationatDistanceAlongSpline (Variables|Default|GetSpline) (* (Spline|GetSplineLength (Variables|Default|GetSpline)) (Variables|Default|GetProgress)) "World")) 384.0))))

(fn OnRep_MapPingLocation
  (Transformation|SetWorldLocation (Variables|Default|GetMapPingCollisionSphere) (Variables|Default|GetMapPingLocation) false false))

(fn ConstructionScript)
