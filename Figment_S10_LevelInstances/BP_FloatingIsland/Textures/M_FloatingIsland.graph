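# Translucent UI material that draws a minimap floating island with a vortex, debris and optional glow around it.
# The island texture and two samples of an FX texture read separate UV channels, alongside center-scale and rotator nodes for sizing and spin.
# FX channels combine into tinted vortex and debris layers, with island alpha masking their color.
# Static switches toggle a two-color vortex gradient and the radial glow's color and opacity contributions.
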Material /Figment_S10_LevelInstances/BP_FloatingIsland/Textures/M_FloatingIsland
graph {
 properties {blendMode=BLEND_Translucent, materialDomain=MD_UI, numCustomizedUVs=4, pixelDepthOffsetMode=PDOM_Legacy, usedShadingModels=Default Lit}
 nodes [
  n1 = ScalarParameter() {defaultValue=0.4, parameterName=IslandScale}
  MaterialFunctionCall(Texture Scale=MaterialFunctionCall(X=@n1, Y=@n1) {materialFunction=&MakeFloat2}.Result) {materialFunction=&ScaleUVsByCenter}
  n2 = ScalarParameter() {defaultValue=0.48, parameterName=ShadowScale_Max}
  MaterialFunctionCall(Texture Scale=MaterialFunctionCall(X=@n2, Y=@n2) {materialFunction=&MakeFloat2}.Result) {materialFunction=&ScaleUVsByCenter}
  Rotator() {speed=-1.7}
  Rotator() {speed=-0.4}
  n3 = ScalarParameter() {defaultValue=1, parameterName=Glow_Density}
  n4 = TextureSample(UVs=TextureCoordinate() {coordinateIndex=1}) {samplerSource=SSM_Clamp_WorldGroupSettings, texture=&/T_UI_Minimap_Island}
  n5 = OneMinus(None=@n4.A)
  n6 = Multiply(
   A=Multiply(A=MaterialFunctionCall(Density=@n3) {materialFunction=&RadialGradientExponential}.RadialGradientExponential)
   B=@n5
  )
  n7 = VectorParameter() {parameterName=Glow_Color, defaultValue=<rgba 1 0 0.8783 1>}
  n8 = TextureSample(UVs=TextureCoordinate() {coordinateIndex=2}) {samplerSource=SSM_Clamp_WorldGroupSettings, texture=&/T_UI_Minimap_FX_1}
  n9 = TextureSample(UVs=TextureCoordinate() {coordinateIndex=3}) {samplerSource=SSM_Clamp_WorldGroupSettings, texture=&T_UI_Minimap_FX_1}
  n10 = Multiply(A=Add(A=@n8.G, B=@n9.G), B=ScalarParameter() {defaultValue=1, parameterName=Debris_Brightness}) {constB=2}
  n11 = Multiply(A=@n10, B=VectorParameter() {parameterName=Debris_Color, defaultValue=<rgba 0.3199 0.9375 0.2824 1>}.RGB)
  n12 = VectorParameter() {parameterName=VortexColor_A, defaultValue=<rgba 0.7249 1 0.9801 1>}
  n13 = Add(A=Multiply(A=Multiply(A=@n8.R, B=@n9.R), B=@n9.B), B=Multiply(A=@n8.B, B=@n9.B))
  n14 = StaticSwitchParameter(
   True=LinearInterpolate(A=@n12.RGB, B=VectorParameter() {parameterName=VertexColor_B, defaultValue=<rgba 0.6206 0.8127 1 1>}.RGB, Alpha=@n13)
   False=@n12.RGB
  ) {parameterName=UseGradientVortex}
  n15 = Add(A=@n13, B=@n8.R)
  n16 = Multiply(A=@n5, B=Multiply(A=@n15, B=ScalarParameter() {defaultValue=0.7373, parameterName=VortexBrightness}))
  n17 = ScalarParameter() {defaultValue=0.4349, parameterName=VortexOpacity}
  n18 = Add(A=Add(A=Multiply(A=@n4.RGB, B=@n4.RGB), B=Multiply(A=@n5, B=@n11)), B=Multiply(A=Multiply(A=@n14, B=@n16), B=@n17))
  StaticSwitchParameter(True=Add(A=Multiply(A=@n6, B=@n7.RGB), B=@n18), False=@n18) {parameterName=GlowEnabled_Color}
  n19 = Multiply(A=@n15, B=@n17)
  n20 = Clamp(None=Add(A=Add(A=@n4.A, B=@n19), B=@n10))
  StaticSwitchParameter(True=Add(A=@n6, B=@n20), False=@n20) {parameterName=GlowEnabled_Opacity}
 ]
}
